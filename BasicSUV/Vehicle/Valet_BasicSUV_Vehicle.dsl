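; Basic SUV Valet vehicle Blueprint, layering driving feedback and lighting effects onto its parent class.
; Gear changes trigger force feedback and camera shake, plus an FOV kick on upshifts; boost adds force feedback and a looping gameplay cue.
; Headlight toggles lazily spawn or toggle Niagara lens flares, with flare count derived from existing mesh sockets.
; A gameplay tag flags electric exhaust FX, and an OnRep applies headlight color on clients.
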(event EventBeginPlay
  ;[1 1 1] Begin play events
  (|CallParentFunction)
  (Class|FortAsyncActionGameStateReady|GameStateReadyAsync
    (:OnReady
      (Vehicle|BindEventtoOnCurrentGearChangedDelegate (EventDispatchers|CreateEvent :!event Custom|HandleGearChange)))))

(event Custom|HandleGearChange
  (params (NewGear :type Integer))
  ;[1 1 1] Event ----> Handle any gear change fx/force feedback
  (bind _upshift (CallFunction|HandleShifting NewGear))
  ;; HandleFOVDuringShift
  (if _upshift
    (Variables|Camera|SetCameraFOVOffset -5.0)
    (Utilities|FlowControl|Delay 0.2)
    (Variables|Camera|SetCameraFOVOffset 0.0)))

(event EventPickRandomMaterial)

(event EventOnBoostStarted
  (bind _driverpawn (Variables|ValetDefaults|GetDriverPawn))
  (bind _returnvalue (Fortnite|GetFortPlayerControllerfromActor _driverpawn))
  ;[1 1 1] Boost Event > VFX > FF > State Set
  (|CallParentFunction)
  (Utilities|IsValid _returnvalue
    (:Is%20Valid
      (Variables|Default|SetbHasBoosted true)
      (Game|Feedback|ClientPlayForceFeedback _returnvalue (Variables|Default|GetFF_CatchUpBoost) "None" false false false)
      (Projectile|AddGameplayCue%28Looping%29Local _returnvalue (Variables|Default|GetGC_CatchUpBoost) (Ability|GameplayCue|MakeGameplayCueParameters 0.0)))))

(event EventOnBoostFinished
  (bind _driverpawn (Variables|ValetDefaults|GetDriverPawn))
  (bind _returnvalue (Fortnite|GetFortPlayerControllerfromActor _driverpawn))
  ;[1 1 1] Boost End Event > Remove VFX > Reset State || Delay is to make sure FOV reset back to normal so shifting doesnt affect it
  (|CallParentFunction)
  (Utilities|IsValid _returnvalue
    (:Is%20Valid
      (Projectile|RemoveGameplayCue%28Looping%29Local _returnvalue (Variables|Default|GetGC_CatchUpBoost) (Ability|GameplayCue|MakeGameplayCueParameters 0.0))
      (Utilities|FlowControl|Delay 1.0)
      (Variables|Default|SetbHasBoosted false))))

(event EventHandleOnExplode
  (params (LastDamageInstigator :type Controller%20Object%20Reference) (Vehicle :type Actor%20Object%20Reference))
  ;[1 1 1] Event ----> Handle Lens Flare on explosion and headlight change
  (|CallParentFunction LastDamageInstigator Vehicle)
  (bind _spawned_headlights_niagara (Variables|Default|GetSpawned%20Headlights%20Niagara)
    (:then
      (Components|Activation|Deactivate _spawned_headlights_niagara))
    (:else)))

(event OnLightEnabledChanged(ValetLightSpawnerComponent)
  (params (enabled :type Boolean))
  ;[1 1 1] Event ----> Handle Lens Flare on explosion and headlight change
  (if enabled
    (bind _spawned_headlights_niagara (Variables|Default|GetSpawned%20Headlights%20Niagara)
      (:then
        (Components|Activation|Activate _spawned_headlights_niagara false)
        (Rendering|SetVisibility _spawned_headlights_niagara true false))
      (:else
        (bind _returnvalue (Niagara|SpawnSystemAttached (Variables|FX|GetHeadlights%20Niagara%20System) (Variables|Vehicle|GetSkeletalMesh) "root"))
        (Variables|Default|SetSpawned%20Headlights%20Niagara _returnvalue)
        (CallFunction|SetNumberOfFlarestoHeadlights _returnvalue)))
    (else
      (bind _spawned_headlights_niagara_2 (Variables|Default|GetSpawned%20Headlights%20Niagara)
        (:then
          (Components|Activation|Deactivate _spawned_headlights_niagara_2)
          (Rendering|SetVisibility _spawned_headlights_niagara_2 false false))
        (:else)))))

(event Vehicle|Cosmetics|EventOnVehicleCosmeticsFinished
  (bind _spawned_headlights_niagara (Variables|Default|GetSpawned%20Headlights%20Niagara))
  ;[1 1 1] Wait to load VCID Info and check if contains tag FUEL.ELECTRIC to set to native exhaust component
  (|CallParentFunction)
  (CallFunction|IsElectricCar)
  (Utilities|IsValid _spawned_headlights_niagara
    (:Is%20Valid
      (CallFunction|SetNumberOfFlarestoHeadlights _spawned_headlights_niagara))))

(event EventOnPawnEnterVehicle
  (params (PlayerPawn :type Fort%20Player%20Pawn%20Object%20Reference) (SeatIdx :type Integer))
  ;[1 1 1] Wait to load VCID Info and check if contains tag FUEL.ELECTRIC to set to native exhaust component
  (|CallParentFunction PlayerPawn SeatIdx)
  (CallFunction|IsElectricCar))

(fn CleanupDriftGE
  (GameplayEffects|RemoveActiveGameplayEffectbySourceEffect (Ability|GetAbilitySystemComponent (Variables|ValetDefaults|GetDriverPawn)) (Variables|Drift|GetDriftingTag) 0 -1)
  (for _array_element (Vehicle|GetAllPassengers)
    (Utilities|IsValid _array_element
      (:Is%20Valid
        (GameplayEffects|RemoveActiveGameplayEffectbySourceEffect (Ability|GetAbilitySystemComponent _array_element) (Variables|Drift|GetDriftingTag) 0 -1))))
  (return))

(fn HandleShifting
  (params (NewGear :type Integer))
  (bind _driverpawn (Variables|ValetDefaults|GetDriverPawn))
  (bind _returnvalue (Fortnite|GetFortPlayerControllerfromActor _driverpawn))
  (bind _driverpawn_2 (Variables|ValetDefaults|GetDriverPawn))
  (bind _returnvalue_2 (Fortnite|GetFortPlayerControllerfromActor _driverpawn_2))
  (Utilities|IsValid (Variables|ValetDefaults|GetDriverPawn)
    (:Is%20Valid
      (if (> NewGear (Variables|ShiftingGears|GetCurrentGearAudio))
        (Development|PrintString "UpShift")
        (Game|Feedback|ClientPlayForceFeedback _returnvalue_2 (Variables|ShiftingGears|GetFF_Upshift) "None" false false false)
        (Game|Feedback|ClientStartCameraShake _returnvalue_2 (Variables|ShiftingGears|GetUpShiftCameraShake) 1.0 "CameraLocal" (Transformation|GetActorRotation _driverpawn_2))
        (Variables|ShiftingGears|SetCurrentGearAudio NewGear)
        (return true)
        (else
          (Development|PrintString "Downshift")
          (Game|Feedback|ClientPlayForceFeedback _returnvalue (Variables|ShiftingGears|GetFF_Downshift) "None" false false false)
          (Game|Feedback|ClientStartCameraShake _returnvalue (Variables|ShiftingGears|GetDownShiftCameraShake) 1.0 "CameraLocal" (Transformation|GetActorRotation))
          (Variables|ShiftingGears|SetCurrentGearAudio NewGear)
          (return false))))))

(fn InitializeHeadlightAndLensFlareFX
  (bind _result (Math|Integer|IncrementInt (Variables|Lights|GetUpdateValetHeadlightColors))
    (:%20%20
      (Variables|Lights|SetUpdateValetHeadlightColors _result)
      ;[1 1 1] Override value based on child needs
      (bind _spawned_headlights_niagara (Variables|Default|GetSpawned%20Headlights%20Niagara)
        (:then
          (Niagara|SetNiagaraVariable(Int32) _spawned_headlights_niagara "SedanColor" (Variables|Lights|GetLensFlare_SedanColor)))
        (:else)))))

(fn InitializeTailLightColor
  (Niagara|SetNiagaraVariable(Int32) (Variables|Valet|RoadAndExhaust|Niagara|GetNativeComp_RoadAndExhaust) "SedanColor" (Variables|Default|GetRoadAndExhaust_TailLight_SedanColor)))

(fn IsElectricCar
  (bind _returnvalue (GameplayTags|HasMatchingGameplayTag "(TagName=\"Vehicle.Fuel.Electric\")"))
  (Niagara|SetNiagaraVariable(Bool) (Variables|Valet|RoadAndExhaust|Niagara|GetNativeComp_RoadAndExhaust) "IsElectric" _returnvalue)
  (return _returnvalue))

(fn OnRep_UpdateValetHeadlightColors
  (if (not (Networking|IsServer))
    ;[1 1 1] Override value based on child needs
    (ValetLights|SetValetLightColor (Variables|Default|GetValetLightSpawnerComponent) (Variables|Lights|GetHeadLightColor))))

(fn SetNumberOfFlaresToHeadlights
  (params (Niagara%20system :type Niagara%20Particle%20System%20Component%20Object%20Reference))
  (bind _existingsocketslist (LocalVariables|GetExistingSocketsList))
  (bind _returnvalue (Utilities|Array|Length _existingsocketslist))
  (bind _asfort_vehicle_skel_mesh_component (Utilities|Casting|CastToFortVehicleSkelMeshComponent (Variables|Vehicle|GetSkeletalMesh)))
  (for _array_element (LocalVariables|GetFilteredSocketsList)
    (if (Transformation|DoesSocketExist _asfort_vehicle_skel_mesh_component _array_element)
      (Utilities|Array|Add (LocalVariables|GetExistingSocketsList) _array_element)
      (Niagara|SetNiagaraVariable(Int32) Niagara%20system "FlaresCount" _returnvalue)
      (Development|PrintString (Utilities|String|Append "Car lensflares Headligth: " (Utilities|String|ToString(Integer) _returnvalue)) true true "(R=1.000000,G=0.000000,B=0.145770,A=1.000000)" 4.0))))

(fn ConstructionScript
  (|CallParentFunction))
